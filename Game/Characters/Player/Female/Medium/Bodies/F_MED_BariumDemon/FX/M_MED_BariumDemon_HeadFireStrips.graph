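# Translucent, unlit, two-sided Niagara mesh-particle material for flame strips on the female BariumDemon player character's head.
# World position offset combines SplineThicken, sine-driven rotation about the particle position, and a directional wobble weighted along UV V.
# Opacity smoothsteps panned noise masked by a V gradient and sphere mask, while particle color feeds the emissive adjustment function.
# A shared FXTime offsets time by a dynamic parameter and vertex blue.
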Material /Game/Characters/Player/Female/Medium/Bodies/F_MED_BariumDemon/FX/M_MED_BariumDemon_HeadFireStrips
graph {
 properties {
  bUsedWithNiagaraMeshParticles=true
  blendMode=BLEND_Translucent
  pixelDepthOffsetMode=PDOM_Legacy
  shadingModel=MSM_Unlit
  twoSided=true
  usedShadingModels=Unlit
 }
 nodes [
  n1 = ScalarParameter() {defaultValue=10, parameterName=Thickness}
  n2 = ScalarParameter() {parameterName=EndThickness}
  n3 = Add(A=NamedRerouteUsage(), B=WorldPosition().XYZ)
  n4 = Transform(None=VectorParameter() {parameterName=RotationAngle, defaultValue=<rgba 0 0 0 1>}.RGB) {transformSourceType=TRANSFORMSOURCE_Local}
  n5 = ComponentMask(None=TextureCoordinate()) {g=true}
  n6 = Sine(None=Add(A=Time(), B=Multiply(A=@n5) {constB=0.3}))
  n7 = ScalarParameter() {parameterName=RotationAmnt}
  n8 = Multiply(
   A=ScalarParameter() {defaultValue=1, parameterName=ShouldRotate}
   B=RotateAboutAxis(
    NormalizedRotationAxis=Normalize(VectorInput=@n4)
    RotationAngle=Multiply(A=@n6, B=@n7)
    PivotPoint=ParticlePositionWS()
    Position=WorldPosition().XYZ
   )
  )
  n9 = ComponentMask(None=TextureCoordinate()) {g=true}
  n10 = Add(A=ComponentMask(None=TextureCoordinate()) {g=true}, B=NamedRerouteUsage())
  n11 = Multiply(A=Multiply(A=OneMinus(None=@n9), B=Sine(None=@n10)), B=ScalarParameter() {group=WPO_Wobble, parameterName=WobbleAmount})
  n12 = Transform(None=VectorParameter() {group=WPO_Wobble, parameterName=WobbleDir, defaultValue=<rgba 1 0 0 1>}.RGB) {transformSourceType=TRANSFORMSOURCE_Local}
  Add(
   A=Add(
    A=MaterialFunctionCall(WidthBase=@n1, WidthTip=@n2, WorldPosition=@n3) {materialFunction=&SplineThicken}.WorldPositionOffset
    B=NamedRerouteDeclaration(None=@n8) {name=RotationFX, nodeColor=<rgba 0 1 0.8828 1>}
   )
   B=Multiply(A=@n11, B=Normalize(VectorInput=@n12))
  )
  MaterialFunctionCall(Emissive=ParticleColor().RGB) {materialFunction=&MF_FX_EmissiveAdjustments}
  Multiply(A=@n11)
  n13 = Time()
  n14 = DynamicParameter() {paramNames [Param1,Param2,FXTimeOffset,Param4]}
  n15 = VertexColor()
  n16 = ScalarParameter() {parameterName=VertexColorBlue_TimeShift}
  NamedRerouteDeclaration(None=Add(A=Add(A=@n13, B=@n14.FXTimeOffset), B=Multiply(A=@n15.B, B=@n16))) {name=FXTime, nodeColor=<rgba 0.6875 0 1 1>}
  n17 = ScalarParameter() {parameterName=Opacity_Pos}
  n18 = TextureCoordinate()
  n19 = Multiply(A=OneMinus(None=ComponentMask(None=@n18) {g=true})) {constB=2}
  n20 = SphereMask(
   A=TextureCoordinate() {vTiling=0.25}
   B=ComponentMask(None=VectorParameter() {parameterName=EdgeMask, defaultValue=<rgba 0.5 0 0 1>}.RGB) {g=true, r=true}
  ) {attenuationRadius=0.5, hardnessPercent=0}
  n21 = TextureCoordinate()
  n22 = ComponentMask(None=VectorParameter() {parameterName=TextureTiling, defaultValue=<rgba 1 1 0 1>}.RGB) {g=true, r=true}
  n23 = VectorParameter() {parameterName=TextureSpeed, defaultValue=<rgba 1 1 0 1>}
  n24 = TextureSampleParameter2D(UVs=Panner(Coordinate=Multiply(A=@n21, B=@n22), Time=NamedRerouteUsage(), Speed=ComponentMask(None=@n23.RGB) {g=true, r=true}) {speedY=1}) {parameterName=T_MainNoise, samplerType=SAMPLERTYPE_LinearColor, texture=/Game/Items/Traps/Textures/T_Plasma_Flames_02_M}
  SmoothStep(
   Min=@n17
   Max=Add(A=@n17, B=ScalarParameter() {parameterName=Opacity_W})
   Value=Multiply(A=Saturate(None=@n19), B=Multiply(A=@n20, B=@n24.R))
  )
 ]
}
